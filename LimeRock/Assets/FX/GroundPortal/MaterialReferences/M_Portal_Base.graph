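# Base unlit translucent material for a glowing ground portal effect, usable on Niagara mesh particles.
# Three sphere masks shape the edge glow and opacity, with edge color over a background color and depth fade.
# Detail UVs use optional bump-offset parallax, radial mapping, panning and tilt for an emissive texture.
# A sphere-masked panning noise offset feeds only the previous-frame side of PreviousFrameSwitch, distorting motion vectors without moving vertices.
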Material /LimeRock/Assets/FX/GroundPortal/MaterialReferences/M_Portal_Base
graph {
 properties {
  bUsedWithNanite=true
  bUsedWithNiagaraMeshParticles=true
  blendMode=BLEND_Translucent
  pixelDepthOffsetMode=PDOM_Legacy
  shadingModel=MSM_Unlit
  usedShadingModels=Unlit
 }
 nodes [
  MaterialFunctionCall(Emissive=NamedRerouteUsage()) {materialFunction=&MF_FX_EmissiveAdjustments}
  n1 = SphereMask(
   A=TextureCoordinate()
   B=Constant() {r=0.5}
   Radius=ScalarParameter() {defaultValue=0.5, group=Sphere Mask, parameterName=Sphere Mask Radius}
   Hardness=NamedRerouteUsage()
  ) {hardnessPercent=50}
  n2 = SphereMask(
   A=NamedRerouteUsage()
   B=Constant() {r=0.5}
   Radius=ScalarParameter() {defaultValue=1.1629, parameterName=Bump Sphere Mask Radius}
   Hardness=NamedRerouteUsage()
  ) {hardnessPercent=50}
  n3 = SphereMask(
   A=NamedRerouteUsage()
   B=Constant() {r=0.5}
   Radius=ScalarParameter() {defaultValue=0.676, parameterName=Bump Sphere Mask 2 Radius}
   Hardness=ScalarParameter() {defaultValue=0.2452, parameterName=Bump Sphere Mask 2 Hardness}
  ) {hardnessPercent=50}
  n4 = Multiply(
   A=Max(A=OneMinus(None=@n1), B=OneMinus(None=@n2))
   B=LinearInterpolate(B=OneMinus(None=@n3), Alpha=NamedRerouteUsage()) {constA=1, constAlpha=1}
  )
  n5 = Multiply(
   A=Power(Base=@n4, Exp=ScalarParameter() {defaultValue=2, parameterName=Sphere Mask Power})
   B=ScalarParameter() {defaultValue=2, parameterName=Sphere Mask Strenth}
  )
  n6 = OneMinus(None=@n5)
  n7 = LinearInterpolate(Alpha=NamedRerouteUsage()) {constA=100}
  n8 = Saturate(None=Multiply(A=Power(Base=@n6, Exp=@n7) {constExponent=1}, B=NamedRerouteUsage()))
  MaterialFunctionCall(Alpha Threshold=@n8) {materialFunction=&DitherTemporalAA_Fortnite}
  Multiply(
   A=MaterialFunctionCall!(
    InRamp=@n4
    CoreIntensity=ScalarParameter() {defaultValue=0.3439, parameterName=OverAllIntensity}
    DetailBalence=ScalarParameter() {defaultValue=1.5, parameterName=Param_6}
    MaxOutPut=Constant() {r=10}
   ) {materialFunction=/Hermes/Props/ClayTile/ClayTile_DragonGate_B/VFX/MF_MOON_InvertDistanceRamp}.Result
   B=NamedRerouteUsage()
  )
  Multiply(
   A=ScalarParameter() {defaultValue=25, parameterName=Dither Strength}
   B=MaterialFunctionCall() {materialFunction=&DitherTemporalAA_Fortnite}.Result
  )
  n9 = TextureCoordinate()
  n10 = ScalarParameter() {defaultValue=-12.5039, parameterName=Bump Offset Height}
  NamedRerouteDeclaration(None=StaticSwitchParameter(True=@n9, False=BumpOffset(Coordinate=@n9, Height=@n10) {referencePlane=0}) {parameterName=Is Flat}) {name=BumpUV, nodeColor=<rgba 0.6406 0 1 1>}
  n11 = Add(A=NamedRerouteUsage()) {constB=-0.5}
  n12 = VectorParameter() {parameterName=Main_UV, defaultValue=<rgba 3 2.101 -0.1307 0.248>}
  n13 = Time()
  n14 = ScalarParameter() {defaultValue=0.5, parameterName=PanSpeed}
  NamedRerouteDeclaration(None=MaterialFunctionCall!(
    ClampTexture
    UV=MaterialFunctionCall(Vector or UVs=@n11) {materialFunction=&VectorToRadialValue}.Radial Coordinates
    Tex
    InRGB=@n12.RGB
    InA=@n12.RGB
    Time=Multiply(A=@n13, B=@n14) {constB=0.5}
    TimeMult
   ) {materialFunction=/Hermes/Props/ClayTile/ClayTile_DragonGate_B/VFX/MF_MOON_TexPanner_V2}.PannerUV) {name=MainUV, nodeColor=<rgba 0 0.4049 1 1>}
  n15 = MaterialFunctionCall(Float2=NamedRerouteUsage()) {materialFunction=&BreakOutFloat2Components}
  n16 = Multiply(A=@n15.R, B=ScalarParameter() {defaultValue=1, parameterName=UV_Tilt})
  NamedRerouteDeclaration(None=MaterialFunctionCall(X=@n15.R, Y=Add(A=@n15.G, B=@n16)) {materialFunction=&MakeFloat2}.Result) {name=TiltedUV, nodeColor=<rgba 0 0.3848 1 1>}
  n17 = SphereMask(A=TextureCoordinate(), B=Constant() {r=0.5}, Radius=ScalarParameter() {defaultValue=0.2691, parameterName=WPO Sphere Mask Radius}) {hardnessPercent=0}
  n18 = TextureCoordinate()
  n19 = Panner(Coordinate=@n18) {speedX=1, speedY=0.1249}
  n20 = Multiply(
   A=ConstantBiasScale(None=TextureSample(UVs=@n19) {samplerType=SAMPLERTYPE_LinearColor, texture=/Hermes/Props/ClayTile/ClayTile_DragonGate_B/VFX/T_FX_Tile_0138_Moon}.RGB) {bias=-0.5, scale=2}
   B=ScalarParameter() {defaultValue=16.5219, parameterName=WPO Strength}
  )
  NamedRerouteDeclaration(None=PreviousFrameSwitch(Current Frame=Constant(), Previous Frame=Multiply(A=@n17, B=@n20))) {name=OUT_WPO, nodeColor=<rgba 0 0.625 1 1>}
  NamedRerouteDeclaration(None=NamedRerouteUsage()) {name=OUT_EdgeMask, nodeColor=<rgba 0.125 1 0 1>}
  NamedRerouteDeclaration(None=DepthFade(Opacity=@n8, FadeDistance=ScalarParameter() {defaultValue=1000, parameterName=DepthFadeDistance})) {name=OUT_Transparent, nodeColor=<rgba 1 0.8438 0 1>}
  n21 = Constant() {r=0.5}
  n22 = SphereMask(A=TextureCoordinate(), B=@n21, Radius=@n21) {attenuationRadius=0.5, hardnessPercent=75}
  NamedRerouteDeclaration(None=@n22) {name=OUT_SphereMask, nodeColor=<rgba 1 0.8672 0 1>}
  NamedRerouteDeclaration(None=ParticleColor().A) {name=OUT_ParticleAlpha, nodeColor=<rgba 0.6172 1 0 1>}
  NamedRerouteDeclaration(None=VectorParameter() {parameterName=Portal Edge Color, defaultValue=<rgba 0.85 0.1746 1 1>}.RGB) {name=OUT_Color, nodeColor=<rgba 0 0.3672 1 1>}
  NamedRerouteDeclaration(None=LinearInterpolate(A=Constant() {r=0.7}, Alpha=NamedRerouteUsage()) {constB=0}) {name=OUT_Dissolve, nodeColor=<rgba 0.6875 1 0 1>}
  n23 = ScalarParameter() {defaultValue=0.6132, parameterName=Emissive Color Max}
  n24 = ScalarParameter() {defaultValue=-0.2691, parameterName=Emissive Color Min}
  n25 = TextureSampleParameter2D!(UVs=NamedRerouteUsage(), Level=ScalarParameter() {defaultValue=1.5, parameterName=Mip Level}, Apply View MipBias) {
   mipValueMode=TMVM_MipLevel
   parameterName=Main Emissive TX
   samplerType=SAMPLERTYPE_LinearColor
   texture=/Hermes/Props/ClayTile/ClayTile_DragonGate_B/VFX/Tile_Tech_02
  }
  n26 = SquareRoot(None=@n25.G)
  n27 = NamedRerouteUsage()
  NamedRerouteDeclaration(None=Subtract(A=LinearInterpolate(A=@n23, B=@n24, Alpha=@n26), B=OneMinus(None=@n27))) {name=OUT_TextureLerp, nodeColor=<rgba 0.9688 0 1 1>}
  NamedRerouteDeclaration(None=Multiply(A=@n8, B=NamedRerouteUsage())) {name=OUT_Opacity, nodeColor=<rgba 1 0 0.2812 1>}
  n28 = NamedRerouteUsage()
  NamedRerouteDeclaration(None=Add(A=Multiply(A=@n5, B=@n28), B=VectorParameter() {parameterName=BG Color, defaultValue=<rgba 0.0115 0.0061 0.026 1>}.RGB)) {name=OUT_Emissive, nodeColor=<rgba 1 0 0.3047 1>}
  NamedRerouteUsage()
  NamedRerouteUsage()
  NamedRerouteUsage()
  Normalize(VectorInput=Subtract(A=@n18) {constB=0.5})
  OneMinus(None=@n22)
  VectorParameter() {parameterName=Param_1, defaultValue=<rgba 28.7196 0 0 0>}
  VertexInterpolator(VS=@n20)
 ]
}
